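# Sound cue for a one-shot ambient heavy, deep metal drop, part of the Asteria metal stress environment sounds.
# A Random node picks among eleven equally weighted wave variations, feeding a Modulator that varies playback with a widened or shifted pitch range.
# A quality level switch routes that Modulator to its first two slots and leaves the last two empty.
# Distant creak attenuation and the ambient one-shot sound class apply, with bus and submix sends disabled.
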SoundCue /Game/Sounds/Asteria/Environment_Actors/MetalStress/Spot_Metal_Heavy_Deep_Drop_Cue
root=QualityLevel_0
nodes [
 {id=Modulator_0, children [Random_1], properties {pitchMin=0.9}}
 {id=QualityLevel_0, children [Modulator_0,Modulator_0,null,null]}
 {
  id=Random_1
  children [
   WavePlayer_8
   WavePlayer_9
   WavePlayer_10
   WavePlayer_11
   WavePlayer_12
   WavePlayer_13
   WavePlayer_14
   WavePlayer_15
   WavePlayer_16
   WavePlayer_17
   WavePlayer_18
  ]
  properties {weights [1,1,1,1,1,1,1,1,1,1,1]}
 }
 {id=WavePlayer_10, properties {soundWaveAssetPtr=&/Spot_Metal_Heavy_Deep_Drop_03}}
 {id=WavePlayer_11, properties {soundWaveAssetPtr=&/Spot_Metal_Heavy_Deep_Drop_04}}
 {id=WavePlayer_12, properties {soundWaveAssetPtr=&/Spot_Metal_Heavy_Deep_Drop_05}}
 {id=WavePlayer_13, properties {soundWaveAssetPtr=&/Spot_Metal_Heavy_Deep_Drop_06}}
 {id=WavePlayer_14, properties {soundWaveAssetPtr=&/Spot_Metal_Heavy_Deep_Drop_07}}
 {id=WavePlayer_15, properties {soundWaveAssetPtr=&/Spot_Metal_Heavy_Deep_Drop_08}}
 {id=WavePlayer_16, properties {soundWaveAssetPtr=&/Spot_Metal_Heavy_Deep_Drop_09}}
 {id=WavePlayer_17, properties {soundWaveAssetPtr=&/Spot_Metal_Heavy_Deep_Drop_10}}
 {id=WavePlayer_18, properties {soundWaveAssetPtr=&/Spot_Metal_Heavy_Deep_Drop_11}}
 {id=WavePlayer_8, properties {soundWaveAssetPtr=&/Spot_Metal_Heavy_Deep_Drop_01}}
 {id=WavePlayer_9, properties {soundWaveAssetPtr=&/Spot_Metal_Heavy_Deep_Drop_02}}
]
properties {
 attenuationSettings=/Game/Sounds/Attenuation/Fort_TileAmbient_OneShot_Distant_Creak_Att
 bEnableBusSends=false
 bEnableSubmixSends=false
 maxDistance=8200
 soundClassObject=&Fort_Ambient_OneShots
 volumeMultiplier=0.576
 attenuationOverrides {falloffDistance=6000, stereoSpread=10000, attenuationShapeExtents {x=2000}}
}
