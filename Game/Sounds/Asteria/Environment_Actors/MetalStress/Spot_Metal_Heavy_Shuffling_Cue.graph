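# Spot ambient SoundCue that plays a heavy metal shuffling one-shot for Asteria environment metal stress, classed as outdoor ambience.
# A Random node picks one of four equally weighted wave variations, and a Modulator varies its pitch.
# The root Quality Level node feeds that same chain to its first two levels and leaves the other two empty.
# A distant-creak attenuation asset spatializes it, with bus and submix sends disabled.
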SoundCue /Game/Sounds/Asteria/Environment_Actors/MetalStress/Spot_Metal_Heavy_Shuffling_Cue
root=QualityLevel_0
nodes [
 {id=Modulator_0, children [Random_1], properties {pitchMin=0.9}}
 {id=QualityLevel_0, children [Modulator_0,Modulator_0,null,null]}
 {id=Random_1, children [WavePlayer_8, WavePlayer_9, WavePlayer_10, WavePlayer_11], properties {weights [1,1,1,1]}}
 {id=WavePlayer_10, properties {soundWaveAssetPtr=&/Spot_Metal_Heavy_Shuffle_03}}
 {id=WavePlayer_11, properties {soundWaveAssetPtr=&/Spot_Metal_Heavy_Shuffle_04}}
 {id=WavePlayer_8, properties {soundWaveAssetPtr=&/Spot_Metal_Heavy_Shuffle_01}}
 {id=WavePlayer_9, properties {soundWaveAssetPtr=&/Spot_Metal_Heavy_Shuffle_02}}
]
properties {
 attenuationSettings=/Game/Sounds/Attenuation/Fort_TileAmbient_OneShot_Distant_Creak_Att
 bEnableBusSends=false
 bEnableSubmixSends=false
 maxDistance=8200
 soundClassObject=&Fort_Ambient_Outdoors_Athena
 volumeMultiplier=0.501
 attenuationOverrides {falloffDistance=6000, stereoSpread=10000, attenuationShapeExtents {x=2000}}
}
