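# Sound cue for metal-stress environment actors that plays a random one-shot metal groan as spatialized outdoor ambience.
# A quality-level node routes its first two slots to a random pick among pitched-down sheet groans, unmodified sheet groans and distant cave metal wronks.
# Its last two slots use a smaller set of three sheet groans with random pitch modulation.
# Attenuation comes from a shared CuddleHead metal-stress asset.
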SoundCue /Game/Sounds/Asteria/Environment_Actors/MetalStress/Spot_Metal_Sheet_Groans_Cue
root=QualityLevel_0
nodes [
 {id=Modulator_0, children [Random_1], properties {pitchMax=0.9, pitchMin=0.8}}
 {id=Modulator_1, children [Random_3], properties {pitchMin=0.9}}
 {id=QualityLevel_0, children [Random_2,Random_2,Modulator_1,Modulator_1]}
 {id=Random_0, children [WavePlayer_0,WavePlayer_1,WavePlayer_2,WavePlayer_4,WavePlayer_6,WavePlayer_7], properties {weights [1,1,1,1,1,1]}}
 {
  id=Random_1
  children [WavePlayer_8, WavePlayer_9, WavePlayer_10, WavePlayer_11, WavePlayer_12, WavePlayer_13]
  properties {weights [1,1,1,1,1,1]}
 }
 {id=Random_2, children [Modulator_0,Random_1,Random_0], properties {weights [1,1,1]}}
 {id=Random_3, children [WavePlayer_14, WavePlayer_15, WavePlayer_16], properties {weights [1,1,1]}}
 {id=WavePlayer_0, properties {soundWaveAssetPtr=/Game/Sounds/Artemis/POI/Hollow_Base/Spots/Spot_Cave_Distant_Metal_Wronk_1}}
 {id=WavePlayer_1, properties {soundWaveAssetPtr=/Game/Sounds/Artemis/POI/Hollow_Base/Spots/Spot_Cave_Distant_Metal_Wronk_2}}
 {id=WavePlayer_10, properties {soundWaveAssetPtr=/Game/Sounds/Asteria/POI/CuddleHead/Spots/Spot_Metal_Sheet_Excited_Groan_03}}
 {id=WavePlayer_11, properties {soundWaveAssetPtr=/Game/Sounds/Asteria/POI/CuddleHead/Spots/Spot_Metal_Sheet_Excited_Groan_04}}
 {id=WavePlayer_12, properties {soundWaveAssetPtr=/Game/Sounds/Asteria/POI/CuddleHead/Spots/Spot_Metal_Sheet_Excited_Groan_05}}
 {id=WavePlayer_13, properties {soundWaveAssetPtr=/Game/Sounds/Asteria/POI/CuddleHead/Spots/Spot_Metal_Sheet_Excited_Groan_06}}
 {id=WavePlayer_14, properties {soundWaveAssetPtr=/Game/Sounds/Asteria/POI/CuddleHead/Spots/Spot_Metal_Sheet_Excited_Groan_01}}
 {id=WavePlayer_15, properties {soundWaveAssetPtr=/Game/Sounds/Asteria/POI/CuddleHead/Spots/Spot_Metal_Sheet_Excited_Groan_02}}
 {id=WavePlayer_16, properties {soundWaveAssetPtr=&Spot_Metal_Sheet_Excited_Groan_03}}
 {id=WavePlayer_2, properties {soundWaveAssetPtr=/Game/Sounds/Artemis/POI/Hollow_Base/Spots/Spot_Cave_Distant_Metal_Wronk_3}}
 {id=WavePlayer_4, properties {soundWaveAssetPtr=/Game/Sounds/Artemis/POI/Hollow_Base/Spots/Spot_Cave_Distant_Metal_Wronk_5}}
 {id=WavePlayer_6, properties {soundWaveAssetPtr=/Game/Sounds/Artemis/POI/Hollow_Base/Spots/Spot_Cave_Distant_Metal_Wronk_7}}
 {id=WavePlayer_7, properties {soundWaveAssetPtr=/Game/Sounds/Artemis/POI/Hollow_Base/Spots/Spot_Cave_Distant_Metal_Wronk_8}}
 {id=WavePlayer_8, properties {soundWaveAssetPtr=&Spot_Metal_Sheet_Excited_Groan_01}}
 {id=WavePlayer_9, properties {soundWaveAssetPtr=&Spot_Metal_Sheet_Excited_Groan_02}}
]
properties {
 attenuationSettings=/Game/Sounds/Asteria/POI/CuddleHead/Spots/Fort_CuddleHead_Metal_Stress_OneShot_Sphere_Att
 bEnableBusSends=false
 bEnableSubmixSends=false
 maxDistance=7000
 soundClassObject=&Fort_Ambient_Outdoors_Athena
 volumeMultiplier=0.351
 attenuationOverrides {falloffDistance=6000, stereoSpread=10000, attenuationShapeExtents {x=2000}}
}
