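# Sound cue for a huge, long metal stress one-shot, an environmental ambience spot sound routed to the outdoor Athena ambient sound class.
# A quality-level root feeds the first two levels through one shared modulator and leaves the other two empty.
# The modulator varies pitch over an equally weighted random pick of ten wave variants.
# Distant-creak attenuation shapes the long-range falloff, and bus and submix sends are disabled.
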SoundCue /Game/Sounds/Asteria/Environment_Actors/MetalStress/Spot_Metal_Stress_Huge_Long_Int_Cue
root=QualityLevel_0
nodes [
 {id=Modulator_0, children [Random_0], properties {pitchMin=0.9}}
 {id=QualityLevel_0, children [Modulator_0,Modulator_0,null,null]}
 {
  id=Random_0
  children [WavePlayer_0, WavePlayer_1, WavePlayer_2, WavePlayer_3, WavePlayer_4, WavePlayer_5, WavePlayer_6, WavePlayer_7, WavePlayer_12, WavePlayer_13]
  properties {weights [1,1,1,1,1,1,1,1,1,1]}
 }
 {id=WavePlayer_0, properties {soundWaveAssetPtr=&/Spot_Metal_Stress_Huge_Long_01}}
 {id=WavePlayer_1, properties {soundWaveAssetPtr=&/Spot_Metal_Stress_Huge_Long_02}}
 {id=WavePlayer_12, properties {soundWaveAssetPtr=&/Spot_Metal_Stress_Huge_Long_09}}
 {id=WavePlayer_13, properties {soundWaveAssetPtr=&/Spot_Metal_Stress_Huge_Long_10}}
 {id=WavePlayer_2, properties {soundWaveAssetPtr=&/Spot_Metal_Stress_Huge_Long_03}}
 {id=WavePlayer_3, properties {soundWaveAssetPtr=&/Spot_Metal_Stress_Huge_Long_04}}
 {id=WavePlayer_4, properties {soundWaveAssetPtr=&/Spot_Metal_Stress_Huge_Long_05}}
 {id=WavePlayer_5, properties {soundWaveAssetPtr=&/Spot_Metal_Stress_Huge_Long_06}}
 {id=WavePlayer_6, properties {soundWaveAssetPtr=&/Spot_Metal_Stress_Huge_Long_07}}
 {id=WavePlayer_7, properties {soundWaveAssetPtr=&/Spot_Metal_Stress_Huge_Long_08}}
]
properties {
 attenuationSettings=/Game/Sounds/Attenuation/Fort_TileAmbient_OneShot_Distant_Creak_Att
 bEnableBusSends=false
 bEnableSubmixSends=false
 maxDistance=8200
 soundClassObject=&Fort_Ambient_Outdoors_Athena
 volumeMultiplier=1.002
 attenuationOverrides {falloffDistance=6000, stereoSpread=10000, attenuationShapeExtents {x=2000}}
}
