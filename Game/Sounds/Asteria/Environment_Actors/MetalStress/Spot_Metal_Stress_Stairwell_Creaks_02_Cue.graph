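# Sound cue for looping ambient spot effects of metal stress creaks in a stairwell.
# An endless loop waits a random few to many seconds, then plays one of ten creak waves with pitch and volume modulation; all quality levels share that pool.
# A custom attenuation curve silences the sound up close, peaks at mid-range and fades out with distance.
# Overrides also cap concurrency and feed distant reverb, alongside an exterior ambient bus send.
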SoundCue /Game/Sounds/Asteria/Environment_Actors/MetalStress/Spot_Metal_Stress_Stairwell_Creaks_02_Cue
root=Looping_0
nodes [
 {id=Delay_0, children [Modulator_0], properties {delayMax=18, delayMin=3}}
 {id=Looping_0, children [Delay_0]}
 {id=Modulator_0, children [QualityLevel_0], properties {pitchMin=0.8}}
 {id=QualityLevel_0, children [Random_3,Random_3,Random_3,Random_3]}
 {
  id=Random_3
  children [
   WavePlayer_25
   WavePlayer_26
   WavePlayer_27
   WavePlayer_28
   WavePlayer_29
   WavePlayer_30
   WavePlayer_31
   WavePlayer_32
   WavePlayer_33
   WavePlayer_34
  ]
  properties {weights [1,1,1,1,1,1,1,1,1,1]}
 }
 {id=WavePlayer_25, properties {soundWaveAssetPtr=&/Spot_Metal_Stress_Stairwell_Creak_20}}
 {id=WavePlayer_26, properties {soundWaveAssetPtr=&/Spot_Metal_Stress_Stairwell_Creak_11}}
 {id=WavePlayer_27, properties {soundWaveAssetPtr=&/Spot_Metal_Stress_Stairwell_Creak_12}}
 {id=WavePlayer_28, properties {soundWaveAssetPtr=&/Spot_Metal_Stress_Stairwell_Creak_13}}
 {id=WavePlayer_29, properties {soundWaveAssetPtr=&/Spot_Metal_Stress_Stairwell_Creak_14}}
 {id=WavePlayer_30, properties {soundWaveAssetPtr=&/Spot_Metal_Stress_Stairwell_Creak_15}}
 {id=WavePlayer_31, properties {soundWaveAssetPtr=&/Spot_Metal_Stress_Stairwell_Creak_16}}
 {id=WavePlayer_32, properties {soundWaveAssetPtr=&/Spot_Metal_Stress_Stairwell_Creak_17}}
 {id=WavePlayer_33, properties {soundWaveAssetPtr=&/Spot_Metal_Stress_Stairwell_Creak_18}}
 {id=WavePlayer_34, properties {soundWaveAssetPtr=&/Spot_Metal_Stress_Stairwell_Creak_19}}
]
properties {
 bOverrideAttenuation=true
 bOverrideConcurrency=true
 maxDistance=1750
 pitchMultiplier=0.9
 soundClassObject=&Fort_Doors
 volumeMultiplier=0.3
 attenuationOverrides {
  bEnableSubmixSends=true
  distanceAlgorithm=Custom
  falloffDistance=1750
  reverbDistanceMax=2000
  reverbWetLevelMin=0.6
  attenuationShapeExtents {x=0}
  customAttenuationCurve {editorCurveData {keys [
     {arriveTangent=4.4763, interpMode=RCIM_Cubic, leaveTangent=4.4763, tangentMode=RCTM_User, time=0.1, value=0}
     {arriveTangent=3.3333, interpMode=RCIM_Cubic, leaveTangent=3.3333, time=0.3, value=1}
     <at 0.4 1>
     <at 1 0>
    ]}}
  submixSendSettings [{
    bEnableHPFCutoff=false
    bEnableLPFCutoff=false
    disableManualSendClamp=false
    hPFCutoff=20
    lPFCutoff=20000
    maxSendDistance=6000
    maxSendLevel=1
    minSendDistance=400
    minSendLevel=0.6
    sendLevel=0.2
    sendLevelControlMethod=Linear
    soundSubmix=&Fortnite_Reverb_Distant
   }]
 }
 busSends [{
   audioBus=null
   maxSendDistance=2000
   minSendLevel=0.3
   sendLevel=1
   soundSourceBus=&AmbientAudio_Exterior_SourceBus
  }]
 concurrencyOverrides {maxCount=2}
}
